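# Niagara swing trail effect for the Modern Military pickaxe, drawn as ribbons from a single SwingTrail emitter.
# Spawns two anim trails between bone or socket index pairs, with masks selecting between two width and color values.
# Color is scaled by a system time-of-day multiplier.
# Short-lived particles darken and fade along curves, with alpha also scaled by time of day.
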NiagaraSystem /BRCosmetics/Weapons/FORT_Melee/Pickaxe_Modern_Military_3/FX/NS_Pickaxe_ModernMilitary_Trail
dependencies {
 usedDataInterfaces [&NiagaraDataInterfaceColorCurve, &NiagaraDataInterfaceCurve, &NiagaraDataInterfaceSocketReader]
 usedDynamicInputs []
 usedModules [
  &self:SystemSpawnScript.SystemScriptSource.SystemScriptGraph.NiagaraNodeAssignment_1.SetVariables_#1
  &self:SystemSpawnScript.SystemScriptSource.SystemScriptGraph.NiagaraNodeAssignment_3.SetVariables_#2
  &SystemState
 ]
}
emitters {SwingTrail {
  emitterSpawnScript [&FN_RibbonIDOffsetUtility()]
  emitterUpdateScript [
   &EmitterState()
   &FN_SpawnAnimTrail(AnimTrailWidthMode=::FromSecond, Number of Anim Trails=2)
  ]
  particleSpawnScript [
   &FN_InitializeAnimTrailBoneIndecies()
   &FN_InitializeAnimTrailBoneIndecies(AnimTrail Selection=1, First Bone or Socket Index=1, Second Bone or Socket Index=2) {module=FN_InitializeAnimTrailBoneIndecies001}
   &FN_InitializeAnimTrail(
    AnimTrailWidth=&MaskFloat(Mask Source=0, Mask Parameter=?, Mask=0, Masked Float=1, Passthrough Float=0.444)
    Color=&MultiplyLinearColor(
     Linear Color=&MaskLinearColor(
      Mask Source=0
      Mask Parameter=?
      Mask=0
      Masked Linear Color=<rgba 0.0639 0.0711 0.0782 2>
      Passthrough Linear Color=<rgba 0.055 0.2 0.033 0.666>
     )
     Scale Factor=&MakeVector4_FromLinearColor(LinearColor=$System.TODMultiplier)
    )
    Lifetime=0.23
   )
  ]
  particleUpdateScript [
   &SolveForcesAndVelocity(Clamp Velocity=false, Limit Acceleration=false)
   &FN_UpdateAnimTrail() {module=FN_UpdateAnimTrail001}
   &ScaleColor(
    Scale Alpha=&FloatFromCurve(
     FloatCurve=DataInterface() {propertyValues {Curve {keys [{arriveTangent=-4.9584, interpMode=RCIM_Cubic, leaveTangent=-4.9584, tangentMode=RCTM_User, time=0.3, value=1}, <at 0.9 0 RCIM_Cubic>]}}}
     Scale Curve=&FNTimeOfDayScalar(Night Scale=0.9, Morning Scale=1, Evening Scale=0.95)
    )
    Scale RGB=&MakeVectorFromLinearColorRGB(Color=&ColorFromCurve(ColorCurve=DataInterface() {propertyValues {
        AlphaCurve {keys [<at 0 1 RCIM_Cubic>,<at 1 0 RCIM_Cubic>]}
        BlueCurve {keys [<at 0.2 1 RCIM_Cubic>,<at 1 0.125>]}
        GreenCurve {keys [<at 0.2 1 RCIM_Cubic>,<at 1 0.125>]}
        RedCurve {keys [<at 0.2 1 RCIM_Cubic>,<at 1 0.125>]}
       }}))
    ScaleA=true
    ScaleRGB=true
   )
  ]
  rendererClasses [&NiagaraRibbonRendererProperties]
  renderers [{rendererClass=&NiagaraRibbonRendererProperties}]
 }}
summary {
 systemName=NS_Pickaxe_ModernMilitary_Trail
 emitters [{emitterName=SwingTrail, rendererClasses [&NiagaraRibbonRendererProperties]}]
 userVariables [User.RibbonIDOffset {type=NiagaraInt32}]
}
