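# Physics asset for the Pickaxe_Modern_Military_3 two-handed melee skeletal mesh, giving its handle and dangling strap physics bodies.
# Two capsules on the handle bone cover the weapon, and one small capsule sits on each of two dyn_strap bones.
# The constraints form a handle-to-strap chain: the first joint locks both swings and leaves twist free, acting as a hinge, while the second leaves angular motion unlocked.
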PhysicsAsset /BRCosmetics/Weapons/FORT_Melee/Pickaxe_Modern_Military_3/Meshes/SK_Pickaxe_Modern_Military_3_PhysicsAsset
bodies {
 Melee_TwoHanded_Handle [
  {shapeType=Capsule, length=30.3505, radius=4.1493, center=<-0.0003 -2.3318 -20.6254>}
  {shapeType=Capsule, length=116.25, radius=32.25, center=<-0.0002 9.4745 18.3909>}
 ]
 dyn_strap [{shapeType=Capsule, length=1, radius=1}]
 dyn_strap_02 [{shapeType=Capsule, length=8.25, radius=3.5, center=<0.3983 0.0002 -10.7654>, rotation=<rot 2.6752 0 0>}]
}
constraints [
 {
  bone1Name=dyn_strap
  bone2Name=Melee_TwoHanded_Handle
  swing1LimitDegrees=10
  swing1Motion=Locked
  swing2LimitDegrees=0
  swing2Motion=Locked
  twistLimitDegrees=90
  twistMotion=Free
 }
 {bone1Name=dyn_strap_02, bone2Name=dyn_strap, swing1LimitDegrees=5, swing2LimitDegrees=100, twistLimitDegrees=25}
]
