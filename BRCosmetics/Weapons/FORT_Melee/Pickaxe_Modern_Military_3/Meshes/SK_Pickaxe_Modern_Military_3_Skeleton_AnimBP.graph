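# Animation Blueprint for the Pickaxe_Modern_Military_3 melee weapon skeleton that adds secondary motion to its dangling strap.
# Three branches offset dyn_strap from the cached input pose, feeding either a RigidBody simulation in handle-bone space or one of two Trail controllers.
# bRigidBodyEnabled selects the physics branch; otherwise bCanPlayCustomAnimations picks between trail variants with different offsets and strengths.
# The event graph holds only an unconnected update event.
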AnimBlueprint /BRCosmetics/Weapons/FORT_Melee/Pickaxe_Modern_Military_3/Meshes/SK_Pickaxe_Modern_Military_3_Skeleton_AnimBP
target_skeleton=&/SK_Pickaxe_Modern_Military_3_Skeleton
graphs [
 AnimGraph {nodes [
   n1 = "Transform (Modify) Bone - Bone: dyn_strap"(
    ComponentPose=Local To Component(LocalPose=Use cached pose 'PrePoses'())
    Alpha=1
    Translation="2.508,0.123,-0.467"
    Rotation="-65,0.000,20"
    Scale="1,1,1"
   ) {
    rotationMode=BMM_Additive
    rotationSpace=BCS_BoneSpace
    translationMode=BMM_Additive
    translationSpace=BCS_BoneSpace
    alphaBoolBlend {blendOption=Linear}
    boneToModify {boneName=dyn_strap}
    rotation {pitch=-65, roll=20}
    translation=<2.5078 0.123 -0.4671>
   }
   n2 = Trail controller - Bone: dyn_strap_02(ComponentPose=@n1, Alpha=0.5) {
    bInvertChainBoneAxis=true
    chainBoneAxis=Z
    rotationLimits [{limitMax=<180 180 180>}, {limitMax=<180 180 180>}]
    rotationOffsets [<0 0 0>, <0 0 0>]
    trailBone {boneName=dyn_strap_02}
   }
   n3 = Component To Local(ComponentPose=@n2)
   n4 = "Transform (Modify) Bone - Bone: dyn_strap"(
    ComponentPose=Local To Component(LocalPose=Use cached pose 'PrePoses'())
    Alpha=1
    Translation="0.136,-0.003,0.009"
    Rotation="5,0,35"
    Scale="1,1,1"
   ) {
    rotationMode=BMM_Additive
    rotationSpace=BCS_BoneSpace
    translationMode=BMM_Additive
    translationSpace=BCS_BoneSpace
    alphaBoolBlend {blendOption=Linear}
    boneToModify {boneName=dyn_strap}
    rotation {pitch=5, roll=35}
    translation=<0.136 -0.003 0.009>
   }
   n5 = Trail controller - Bone: dyn_strap_02(ComponentPose=@n4, Alpha=0.2) {
    bInvertChainBoneAxis=true
    chainBoneAxis=Z
    rotationLimits [{limitMax=<180 180 180>}, {limitMax=<180 180 180>}]
    rotationOffsets [<0 0 0>, <0 0 0>]
    trailBone {boneName=dyn_strap_02}
   }
   n6 = Component To Local(ComponentPose=@n5)
   n7 = Get bCanPlayCustomAnimations() -> (bCanPlayCustomAnimations:bool)
   n8 = "Transform (Modify) Bone - Bone: dyn_strap"(
    ComponentPose=Local To Component(LocalPose=Use cached pose 'PrePoses'())
    Alpha=1
    Translation="0.358,-0.084,0.479"
    Rotation="0,0,0"
    Scale="1,1,1"
   ) {
    rotationMode=BMM_Additive
    rotationSpace=BCS_BoneSpace
    translationMode=BMM_Additive
    translationSpace=BCS_BoneSpace
    alphaBoolBlend {blendOption=Linear}
    boneToModify {boneName=dyn_strap}
    translation=<0.358 -0.084 0.479>
   }
   n9 = RigidBody(ComponentPose=@n8) {
    bClampLinearTranslationLimitToRefPose=true
    bOverrideWorldGravity=true
    overridePhysicsAsset=&/SK_Pickaxe_Modern_Military_3_PhysicsAsset
    simulationSpace=BaseBoneSpace
    alphaBoolBlend {blendOption=Linear}
    baseBoneRef {boneName=Melee_TwoHanded_Handle}
    componentAppliedLinearAccClamp=<500 500 500>
    componentLinearAccScale=<0.15 0.15 0.15>
    componentLinearVelScale=<0.1 0.1 0.1>
    overrideWorldGravity {z=-980}
   }
   Output Pose(Result=Blend Poses by bool(
     BlendPose_0=Component To Local(ComponentPose=@n9)
     BlendPose_1=Blend Poses by bool(BlendPose_0=@n6, BlendPose_1=@n3, bActiveValue=@n7) {blendType=Linear}
     bActiveValue=Get bRigidBodyEnabled()
    ) {blendType=Linear}) {layerGroup=None}
   AnimGraphNode_SubInput_0 = Input Pose()
   Save cached pose 'PrePoses'(Pose=@AnimGraphNode_SubInput_0) {cacheName=PrePoses}
  ]}
 EventGraph {nodes [Event BlueprintUpdateAnimation()]}
]
variables []
